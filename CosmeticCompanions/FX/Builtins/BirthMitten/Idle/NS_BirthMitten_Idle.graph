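# Idle Niagara effect for the BirthMitten cosmetic companion, drawing blue voice-over-driven halos at the helmet and neck collar medallion.
# Each emitter bursts persistent particles that follow skeletal mesh bones or sockets with a camera offset.
# The helmet halo toggles on bIsVOActive with front-end dim and size modifiers, while the collar halo takes alpha from VOCurve with platform dimming.
# User.fVisCollar indexes opposite visibility-tag arrays to switch between the two.
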NiagaraSystem /CosmeticCompanions/FX/Builtins/BirthMitten/Idle/NS_BirthMitten_Idle
dependencies {
 usedDataInterfaces [
  &NiagaraDataInterfaceArrayInt32
  &NiagaraDataInterfacePlatformSet
  &NiagaraDataInterfaceSkeletalMesh
 ]
 usedDynamicInputs [
  &FN_FrontEndFloatModifier
  &FN_PickFloatDependOnPlatform
 ]
 usedModules [
  &self:SystemSpawnScript.SystemScriptSource.SystemScriptGraph.NiagaraNodeAssignment_3.SetVariables_#1
  &self:SystemSpawnScript.SystemScriptSource.SystemScriptGraph.NiagaraNodeAssignment_0.SetVariables_#2
  &SystemState
 ]
}
emitters {
 Halo_Helmet {
  emitterUpdateScript [
   &EmitterState()
   &SpawnBurst_Instantaneous(Use Spawn Probability=false, UseLoopCountLimit=true)
  ]
  particleSpawnScript [&InitializeParticle(
    Color=&MakeLinearColorFromVectorAndFloat("Vector (RGB)"=<0 0.75 1>, "Float (Alpha)"=&Multiply_Float(A=1, B=$System.FrontEnd_Dim))
    Lifetime=$System.CurrentLoopDuration
    Position Offset=<0 -4 30>
    Sprite Size Mode=::NewEnumerator3
    Uniform Sprite Size=&Multiply_Float(A=20, B=$System.FrontEnd_Size)
    UsePositionOffset=true
    Write Material Random=true
   )]
  particleUpdateScript [
   &ParticleState(Kill Particles When Lifetime Has Elapsed=false)
   /SetVariables(Particles.Color=&MakeCustomLinearColorFromBool(Bool=$User.bIsVOActive, True Color=$Particles.Initial.Color, False Color=<rgba 0 0 0 0>)) {isSetParameters=true, module=SetVariables_#3}
   &CameraOffset(Camera Offset Amount=30)
   &SkeletalMeshLocation(
    Bone Sampling Mode=::NewEnumerator4
    MaskSpawnGroup=false
    Sampled Position Offset:NiagaraPosition=$Particles.Initial.Position
    Skeletal Mesh=$System.Skeletal Mesh Reader
   )
   /SetVariables(Particles.VisibilityTag=&SelectIntFromArray(
     Direct Array Index=&MakeIntFromFloat(Float=$User.fVisCollar, Conversion Mode=::NewEnumerator1)
     Int Selection Array=DataInterface() {propertyValues {IntData [0,1]}}
    )) {isSetParameters=true}
  ]
  rendererClasses [&NiagaraSpriteRendererProperties, &NiagaraMeshRendererProperties]
  renderers [
   {rendererClass=&NiagaraSpriteRendererProperties}
   {rendererClass=&NiagaraMeshRendererProperties, rendererIndex=1}
  ]
 }
 Halo_Neck_Collar_MedalionBone {
  emitterUpdateScript [
   &EmitterState()
   &SpawnBurst_Instantaneous(Use Spawn Probability=false, UseLoopCountLimit=true)
  ]
  particleSpawnScript [&InitializeParticle(
    Color Mode=::NewEnumerator0
    Lifetime=$System.CurrentLoopDuration
    Position Offset=<-0.5 0 1.5>
    Sprite Size Mode=::NewEnumerator3
    Uniform Sprite Size=16
    UsePositionOffset=true
    Write Material Random=true
   )]
  particleUpdateScript [
   &ParticleState(Kill Particles When Lifetime Has Elapsed=false)
   /SetVariables(Particles.Color=&MakeLinearColorFromVectorAndFloat(
     "Vector (RGB)"=&Multiply_VectorByFloat(Vector=<0 10 20>, Float=$System.Platform_Dim)
     "Float (Alpha)"=$User.VOCurve
    )) {isSetParameters=true, module=SetVariables_#3}
   &SkeletalMeshLocation(
    Bone / Socket Index=1
    Bone Sampling Mode=::NewEnumerator4
    MaskSpawnGroup=false
    Sampled Position Offset:NiagaraPosition=$Particles.Initial.Position
    Skeletal Mesh=$System.Skeletal Mesh Reader
   )
   /SetVariables(Particles.VisibilityTag=&SelectIntFromArray(
     Direct Array Index=&MakeIntFromFloat(Float=$User.fVisCollar, Conversion Mode=::NewEnumerator1)
     Int Selection Array=DataInterface() {propertyValues {IntData [1,0]}}
    )) {isSetParameters=true}
   &CameraOffset(Camera Offset Amount=8)
  ]
  rendererClasses [&NiagaraSpriteRendererProperties, &NiagaraMeshRendererProperties]
  renderers [
   {rendererClass=&NiagaraSpriteRendererProperties}
   {rendererClass=&NiagaraMeshRendererProperties, rendererIndex=1}
  ]
 }
}
summary {
 systemName=NS_BirthMitten_Idle
 emitters [
  {emitterName=Halo_Helmet, rendererClasses [&NiagaraSpriteRendererProperties, &NiagaraMeshRendererProperties]}
  {
   emitterName=Halo_Neck_Collar_MedalionBone
   rendererClasses [&NiagaraSpriteRendererProperties, &NiagaraMeshRendererProperties]
  }
 ]
 userVariables [
  User.bIsFrontEnd {type=NiagaraBool}
  User.bIsFrontEndPreview {type=NiagaraBool}
  User.bIsVOActive {type=NiagaraBool}
  User.fVisCollar {type=NiagaraFloat, defaultValue=0, description=""}
  User.VOCurve {type=NiagaraFloat, defaultValue=0, description=""}
 ]
}
